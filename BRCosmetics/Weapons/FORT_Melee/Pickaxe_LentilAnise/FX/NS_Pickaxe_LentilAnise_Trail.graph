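# Niagara swing trail for the LentilAnise pickaxe, with a ribbon trail, blue lightning ribbons and socket-spawned sparks.
# Both ribbon emitters use anim-trail modules, fading over particle and emitter age; the lightning widens and tints by time of day.
# Sparks spawn by distance moved, driven by curl noise, drag and gravity.
# User.SwingSectionIndex selects ribbon lifetimes and, by parity, flips the sparks' sideways force.
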NiagaraSystem /BRCosmetics/Weapons/FORT_Melee/Pickaxe_LentilAnise/FX/NS_Pickaxe_LentilAnise_Trail
dependencies {
 usedDataInterfaces [&NiagaraDataInterfaceCurve, &NiagaraDataInterfaceColorCurve, &NiagaraDataInterfaceSocketReader]
 usedDynamicInputs []
 usedModules [
  &self:SystemSpawnScript.SystemScriptSource.SystemScriptGraph.NiagaraNodeAssignment_0.SetVariables_#1
  &SystemState
 ]
}
emitters {
 Lightning_Trail {
  emitterSpawnScript [&FN_RibbonIDOffsetUtility()]
  emitterUpdateScript [
   &EmitterState()
   &FN_SpawnAnimTrail()
  ]
  particleSpawnScript [
   &FN_InitializeAnimTrailBoneIndecies()
   &FN_InitializeAnimTrail(Lifetime=&MergeHelper_FloatSwitch(0=0.2, 1=0.35, 2=0.2, 3=0.35, Index=$User.SwingSectionIndex, Num=::NewEnumerator2))
  ]
  particleUpdateScript [
   &SolveForcesAndVelocity(Clamp Velocity=false, Limit Acceleration=false)
   &FN_UpdateAnimTrail() {module=FN_UpdateAnimTrail001}
   &ScaleColor(
    Scale Alpha=&FloatFromCurve(
     FloatCurve=DataInterface() {propertyValues {Curve {keys [
         {arriveTangent=-2.4237, interpMode=RCIM_Cubic, leaveTangent=-2.4237, tangentMode=RCTM_User, time=0, value=1}
         {arriveTangent=-0.2343, interpMode=RCIM_Cubic, leaveTangent=-0.2343, tangentMode=RCTM_User, time=1, value=0}
        ]}}}
     Scale Curve=&FloatFromCurve(FloatCurve=DataInterface() {propertyValues {Curve {keys [<at 0.15 1 RCIM_Cubic>,<at 0.25 0 RCIM_Cubic>]}}}, CurveIndex=$Emitter.Age)
    )
    Scale RGB=&MakeVectorFromLinearColorRGB(Color=&FNTimeOfDayColor(Night Color=<rgba 0.32 0.4976 0.8 1>, Morning Color=<rgba 0.6 0.6262 0.8 1>, Day Color=<rgba 1 1 1 1>, Evening Color=<rgba 0.6 0.6 0.6 1>))
    ScaleA=true
    ScaleRGB=true
   )
   /SetVariables(
    Particles.GlowColor=&MultiplyLinearColor(
     Linear Color=$Particles.Color
     Scale Factor=&MakeVector4_FromLinearColor(LinearColor=&ColorFromCurve(ColorCurve=DataInterface() {propertyValues {
         AlphaCurve {keys [<at 0.0557 1>]}
         BlueCurve {keys [<at 0.0766 2>,<at 0.5082 1>,<at 1 0.5 RCIM_Cubic>]}
         GreenCurve {keys [<at 0.0766 1.3333>,<at 0.5082 0.5>,<at 1 0.2083 RCIM_Cubic>]}
         RedCurve {keys [<at 0.0766 0>,<at 0.5082 0>,<at 1 0 RCIM_Cubic>]}
        }}))
    )
    Particles.LightningColor:LinearColor=&MultiplyLinearColor(
     Linear Color=$Particles.Color
     Scale Factor=&MakeVector4_FromLinearColor(LinearColor=&ColorFromCurve(ColorCurve=DataInterface() {propertyValues {
         AlphaCurve {keys [<at 0.0327 1>]}
         BlueCurve {keys [<at 0.1938 5 RCIM_Cubic>,<at 0.4628 1>,<at 1 1>]}
         GreenCurve {keys [<at 0.1938 2.2995 RCIM_Cubic>,<at 0.4628 0.75>,<at 1 0.5>]}
         RedCurve {keys [<at 0.1938 0.2344 RCIM_Cubic>,<at 0.4628 0>,<at 1 0>]}
        }}))
    )
   ) {isSetParameters=true}
   &ScaleRibbonWidth(Ribbon Width Scale=&FloatFromCurve(FloatCurve=DataInterface() {propertyValues {Curve {
        defaultValue=1
        keys [
         {arriveTangent=0.829, interpMode=RCIM_Cubic, leaveTangent=0.829, tangentMode=RCTM_User, time=0, value=1}
         {interpMode=RCIM_Cubic, tangentMode=RCTM_User, time=1, value=1.5}
        ]
       }}}))
  ]
  rendererClasses [&NiagaraRibbonRendererProperties]
  renderers [
   {rendererClass=&NiagaraRibbonRendererProperties}
   {rendererClass=&NiagaraRibbonRendererProperties, rendererIndex=1}
  ]
 }
 Sparks {
  emitterUpdateScript [
   &EmitterState(Life Cycle Mode=::NewEnumerator1, Loop Behavior=::NewEnumerator1, Loop Duration=0.25, UseLoopDelay=false)
   &SpawnPerUnit(
    Module.Use Max Movement Threshold=true
    Module.Use Movement Tolerance=false
    Spawn Probability=1
    Spawn Spacing=25
    Use Spawn Probability=false
   )
  ]
  particleSpawnScript [
   &InitializeParticle(
    Color=<rgba 0 0.5514 1 1>
    Lifetime=&RandomRangeFloat(Minimum=0.25, Maximum=0.8)
    Sprite Size Max=<1 7.5>
    Sprite Size Min=<2 2.5>
    Sprite Size Mode=::NewEnumerator2
    UsePositionOffset=false
    Write Material Random=true
   )
   &SocketLocation(
    "Inherit Velocity (Scale)"=2
    Inherited Velocity Speed Limit=100
    Interpolate Socket Sampling=true
    SamplePosOffset=true
    Socket Index=1
    Socket Reader=$System.Socket Reader
    Socket Sampling Mode=::NewEnumerator3
    Velocity Sampling:ENiagara_AttributeSamplingApplyOutput=::NewEnumerator0
    VelocitySpeedLimit=true
   )
   &AddVelocity(Velocity=<0 0 100>, Velocity Speed Scale=0.2)
   &BoxLocation(Box Size=<20 1 1>, MaskSpawnGroup=false, Module.Override Seed=false, Rotate Box=false, Surface Only=false)
   &ApplyInitialForces()
  ]
  particleUpdateScript [
   &ParticleState()
   &ScaleColor(
    Scale Mode=::NewEnumerator1
    Scale RGBA=&MakeVector4_FromLinearColor(LinearColor=&MultiplyLinearColor(
      Linear Color=&ColorFromCurve(ColorCurve=DataInterface() {propertyValues {
         AlphaCurve {keys [<at 0 1 RCIM_Cubic>,<at 0.1601 1>,<at 1 0 RCIM_Cubic>]}
         BlueCurve {keys [<at 0 1>,<at 0.3998 0.5029>,<at 0.9826 0.01 RCIM_Cubic>]}
         GreenCurve {keys [<at 0 1>,<at 0.3998 0.22>,<at 0.9826 0.0097 RCIM_Cubic>]}
         RedCurve {keys [<at 0 0.4793>,<at 0.3998 0>,<at 0.9826 0.0076 RCIM_Cubic>]}
        }})
      Scale Factor=&MakeFloat4FromVectorAndFloat("Vector (RGB)"=&VectorFromFloat(Value=5), "Float (Alpha)"=1)
     ))
    ScaleRGBA=true
   )
   /SetVariables(Particles.Glow03_Color=&MultiplyLinearColor(
     Linear Color=$Particles.Color
     Scale Factor=&MakeFloat4FromVectorAndFloat("Vector (RGB)"=&VectorFromFloat(Value=0.4), "Float (Alpha)"=1)
    )) {isSetParameters=true}
   /SetVariables(Particles.Glow03_Size=&Multiply_Vector2D_ByFloat(
     Float=&FloatFromCurve(
      FloatCurve=DataInterface() {propertyValues {Curve {keys [
          <at 0 0 RCIM_Cubic>
          <at 0.2 1.3 RCIM_Cubic>
          {arriveTangent=-0.1423, interpMode=RCIM_Cubic, leaveTangent=-0.1423, tangentMode=RCTM_User, time=1, value=0}
         ]}}}
      Scale Curve=2.5
     )
     Vector2D=<0.5 1>
    )) {isSetParameters=true}
   &ScaleColor(Scale RGB=&VectorFromFloat(Value=4), ScaleA=true, ScaleRGB=true) {module=ScaleColor001}
   &ScaleSpriteSize(Scale Factor=&Vector2DFromFloat(Value=&FloatFromCurve(
      FloatCurve=DataInterface() {propertyValues {Curve {keys [
          <at 0 0 RCIM_Cubic>
          {interpMode=RCIM_Cubic, leaveTangent=-3.0296, tangentMode=RCTM_Break, time=0.2, value=1.2}
          {arriveTangent=0.008, interpMode=RCIM_Cubic, leaveTangent=0.008, tangentMode=RCTM_User, time=1, value=0}
         ]}}}
      Scale Curve=0.5
     ))) {module=ScaleSpriteSize001}
   &LinearForce(Force=&Lerp_Vector(
     A=<-400 0 0>
     B=<400 0 0>
     Alpha=&MakeFloatFromInt(Int=&ModuloInt(A=$User.SwingSectionIndex, B=2))
    ))
   &GravityForce(Gravity=<0 0 -50>)
   &CurlNoiseForce(Noise Frequency=300, Noise Strength=1000, Pan Noise=true, Pan Noise Field=<0 0 1>, Random Seed=10)
   &Drag(Drag=6, Use Linear Drag=true, Use Rotational Drag=false)
   &SolveForcesAndVelocity(Clamp Velocity=false, Limit Acceleration=false)
  ]
  rendererClasses [&NiagaraSpriteRendererProperties]
  renderers [
   {rendererClass=&NiagaraSpriteRendererProperties}
   {rendererClass=&NiagaraSpriteRendererProperties, rendererIndex=1}
  ]
 }
 Trail {
  emitterSpawnScript [/SetVariables(Emitter.Rand:NiagaraFloat=&RandomRangeFloat(Minimum=0, Maximum=1)) {isSetParameters=true}]
  emitterUpdateScript [
   &EmitterState()
   &FN_SpawnAnimTrail()
  ]
  particleSpawnScript [
   &FN_InitializeAnimTrailBoneIndecies()
   /SetVariables(Particles.RibbonFacing=/Game/Effects/Niagara/DynamicInputs/Engine/FN_GetSystemAxis(Axis:ENiagaraOrientationAxis=::XAxis, Flip Axis:NiagaraBool=false)) {isSetParameters=true}
   &FN_InitializeAnimTrail(Lifetime=&MergeHelper_FloatSwitch(0=0.25, 1=0.35, 2=0.25, 3=0.35, Index=$User.SwingSectionIndex, Num=::NewEnumerator2))
   &DynamicMaterialParameters(
    Index 0 Param 1=$Emitter.Rand
    Module.Index0Param0Write=true
    Module.Index0Param1Write=true
    Module.Index0Param2Write=true
    Module.Index0Param3Write=true
   )
  ]
  particleUpdateScript [
   &FN_UpdateAnimTrail()
   &SolveForcesAndVelocity(Clamp Velocity=false, Limit Acceleration=false)
   &ScaleColor(
    Scale Alpha=&FloatFromCurve(
     FloatCurve=DataInterface() {propertyValues {Curve {keys [
         {arriveTangent=-2.4237, interpMode=RCIM_Cubic, leaveTangent=-2.4237, tangentMode=RCTM_User, time=0, value=1}
         {arriveTangent=-0.2343, interpMode=RCIM_Cubic, leaveTangent=-0.2343, tangentMode=RCTM_User, time=1, value=0}
        ]}}}
     Scale Curve=&FloatFromCurve(FloatCurve=DataInterface() {propertyValues {Curve {keys [<at 0.2 1 RCIM_Cubic>,<at 0.4 0 RCIM_Cubic>]}}}, CurveIndex=$Emitter.Age)
    )
    Scale RGB=<2 2 2>
    ScaleA=true
    ScaleRGB=true
   )
  ]
  rendererClasses [&NiagaraRibbonRendererProperties]
  renderers [{rendererClass=&NiagaraRibbonRendererProperties}]
 }
}
summary {
 systemName=NS_Pickaxe_LentilAnise_Trail
 emitters [
  {emitterName=Trail, rendererClasses [&NiagaraRibbonRendererProperties]}
  {emitterName=Sparks, rendererClasses [&NiagaraSpriteRendererProperties]}
  {emitterName=Lightning_Trail, rendererClasses [&NiagaraRibbonRendererProperties]}
 ]
 userVariables [
  User.RibbonIDOffset {type=NiagaraInt32}
  User.SwingSectionIndex {type=NiagaraInt32}
 ]
}
